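# Translucent unlit Niagara sprite material for the IronLilac emote VFX, drawing an ink circle brush stroke.
# A rotated angle gradient passes through a SmoothStep driven by StrokePos and StepMax dynamic parameters, with the ink texture's green channel breaking up the edge.
# That mask multiplies the red channel and particle alpha, then gets depth fade.
# Particle color times EmissiveIntensity feeds emissive adjustments; a Debug switch animates the stroke with time.
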Material /BRCosmetics/Animation/Game/MainPlayer/Emotes/IronLilac/VFX/Materials/M_Emote_IronLilac_BrushStroke
graph {
 properties {
  bAutomaticallySetUsageInEditor=false
  bUsedWithNiagaraSprites=true
  blendMode=BLEND_Translucent
  shadingModel=MSM_Unlit
  usedShadingModels=Unlit
 }
 nodes [
  n1 = ParticleColor()
  n2 = TextureSampleParameter2D() {parameterName=Ink Texture, texture=/BRCosmetics/Animation/Game/MainPlayer/Emotes/IronLilac/VFX/Textures/T_Emote_IronLilac_InkCircle}
  n3 = Sine(None=Time())
  n4 = DynamicParameter() {paramNames [StrokePos, EmissiveIntensity, StepMax, Param4]}
  n5 = Saturate(None=Add(A=StaticSwitchParameter(True=@n3, False=@n4.StrokePos) {parameterName=Debug}, B=Multiply(A=@n2.G) {constB=0.2}))
  n6 = Add(A=@n2.G, B=@n4.StepMax)
  n7 = SmoothStep(
   Min=@n5
   Max=Add(A=@n5, B=Saturate(None=@n6))
   Value=MaterialFunctionCall!(UV, Rotate=Constant() {r=-0.4}) {materialFunction=/AudioWidgets/AudioMaterialSlate/Functions/MF_AngleGradient}.Result
  ) {constMin=0.1}
  DepthFade(
   Opacity=Multiply(A=@n1.A, B=Multiply(A=@n2.R, B=@n7))
   FadeDistance=ScalarParameter() {defaultValue=10, parameterName=DepthFadeDistance, sliderMax=50}
  )
  MaterialFunctionCall(Emissive=Multiply(A=@n4.EmissiveIntensity, B=@n1.RGB)) {materialFunction=&MF_FX_EmissiveAdjustments}
 ]
}
