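# Translucent unlit deferred decal material for the kanji effect of the IronLilac emote, driven by Niagara dynamic parameters and decal color.
# Kanji texture green and blue channels plus Reveal index a curve atlas, whose alpha animates the red-channel mask unless Show All overrides it.
# Erode thresholds the texture's alpha to dissolve the glyph, and a depth fade offset by a custom-depth versus scene-depth comparison scales opacity.
# A separate tiled-noise erosion branch ends unconnected.
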Material /BRCosmetics/Animation/Game/MainPlayer/Emotes/IronLilac/VFX/Materials/M_Emote_IronLilac_Kanji_Decal
graph {
 properties {
  bAutomaticallySetUsageInEditor=false
  bDisableDepthTest=true
  bUsedWithNiagaraSprites=true
  blendMode=BLEND_Translucent
  materialDomain=MD_DeferredDecal
  shadingModel=MSM_Unlit
  twoSided=true
  usedShadingModels=Unlit
 }
 nodes [
  Constant() {r=1}
  n1 = DecalColor()
  n2 = Constant() {r=1}
  n3 = DynamicParameter() {defaultValue=<rgba -0.5 1 0 1>, paramNames [Reveal, Show All, Erode, Param4]}
  n4 = Add(A=Power(Base=@n3.Show All))
  MaterialFunctionCall(Emissive=Multiply(A=@n1.RGB, B=If(A=@n3.Show All, "A > B"=@n2, "A < B"=@n4) {constB=0.5})) {materialFunction=&MF_FX_EmissiveAdjustments}
  n5 = TextureSampleParameter2D() {
   parameterName=Kanji Texture
   samplerType=SAMPLERTYPE_LinearColor
   texture=/BRCosmetics/Animation/Game/MainPlayer/Emotes/IronLilac/VFX/Textures/T_Emote_IronLilac_KiChi
  }
  n6 = Constant() {r=1}
  n7 = Multiply(A=@n5.G, B=ScalarParameter() {defaultValue=1.25, parameterName=Reveal Spacing, sliderMax=2})
  n8 = Multiply(A=Power(Base=@n5.B)) {constB=0.1}
  n9 = CurveAtlasRowParameter(CurveTime=Add(A=Add(A=@n7, B=@n8) {constB=0}, B=@n3.Reveal)) {
   atlas=&/CA_Emote_IronLilac
   curve=&/C_Emote_IronLilac
  }
  n10 = Multiply(A=@n1.A, B=Multiply(A=Power(Base=@n5.R), B=If(A=@n3.Show All, "A > B"=@n6, "A < B"=@n9.A) {constB=0.5}))
  n11 = DynamicParameter() {defaultValue=<rgba -0.5 1 0 1>, paramNames [Reveal, Show All, Erode, Param4]}
  n12 = Saturate(None=@n11.Erode)
  n13 = Add(A=@n11.Erode, B=Constant() {r=0.5})
  n14 = TextureSampleParameter2D() {
   parameterName=Kanji Texture
   samplerType=SAMPLERTYPE_LinearColor
   texture=&T_Emote_IronLilac_KiChi
  }
  n15 = Saturate(None=Subtract(A=SceneTexture() {sceneTextureId=PPI_CustomDepth}.Color, B=SceneTexture() {sceneTextureId=PPI_SceneDepth}.Color))
  n16 = Add(
   A=Subtract(A=SceneDepth().R, B=PixelDepth().R)
   B=LinearInterpolate(A=ScalarParameter() {defaultValue=50, parameterName=Fade Offset}, Alpha=OneMinus(None=@n15)) {constB=0}
  )
  n17 = ScalarParameter() {defaultValue=50, parameterName=Fade Distance}
  Multiply(A=Multiply(A=@n10, B=SmoothStep(Min=@n12, Max=@n13, Value=@n14.A)), B=Saturate(None=Divide(A=@n16, B=@n17)))
  ParticleColor()
  n18 = DynamicParameter() {defaultValue=<rgba -0.5 1 0 1>, paramNames [Reveal, Show All, Erode, Param4]}
  n19 = TextureSampleParameter2D(UVs=Multiply(A=TextureCoordinate(), B=ScalarParameter() {defaultValue=3, parameterName=Erosion Tile, sliderMax=10}) {constB=3}) {
   parameterName=Erosion Texture
   samplerType=SAMPLERTYPE_LinearGrayscale
   texture=/Game/Characters/Player/Female/Medium/Bodies/F_MED_SuperNovaTaro/Textures/LowResBlurredNoise
  }
  n20 = MaterialFunctionCall(In=@n19.R, Contrast=ScalarParameter() {defaultValue=0.25, parameterName=Erosion Contrast, sliderMax=1}) {materialFunction=&CheapContrast}
  Subtract(A=@n10, B=SmoothStep(Min=@n18.Erode, Max=Add(A=@n18.Erode), Value=Saturate(None=@n20.Result)))
 ]
}
