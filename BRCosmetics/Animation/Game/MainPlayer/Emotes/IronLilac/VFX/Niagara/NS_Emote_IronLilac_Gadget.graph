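# Niagara system for the gadget in the IronLilac emote, producing a paint trail and a foam splash sampled from a skeletal mesh.
# PaintTrail spawns short-lived ribbon particles tinted by the system Paint Color, with width and alpha shaped by loop-age and particle-age curves.
# Splash_Foam fires a timed burst of sprites outward from a small cylinder, slowed by drag and gravity.
# Foam carries a visibility tag for its two renderers and scales brightness separately for gameplay and front end.
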NiagaraSystem /BRCosmetics/Animation/Game/MainPlayer/Emotes/IronLilac/VFX/Niagara/NS_Emote_IronLilac_Gadget
dependencies {
 usedDataInterfaces [&NiagaraDataInterfaceCurve, &NiagaraDataInterfaceArrayInt32, &NiagaraDataInterfaceSkeletalMesh]
 usedDynamicInputs []
 usedModules [
  &self:SystemSpawnScript.SystemScriptSource.SystemScriptGraph.NiagaraNodeAssignment_0.SetVariables_#1
  &self:SystemSpawnScript.SystemScriptSource.SystemScriptGraph.NiagaraNodeAssignment_1.SetVariables_#2
  &SystemState
 ]
}
emitters {
 PaintTrail {
  emitterUpdateScript [
   &EmitterState(Life Cycle Mode=::NewEnumerator1, Loop Behavior=::NewEnumerator1, Loop Delay=0.1, Loop Duration=1.5, UseLoopDelay=true)
   &SpawnRate(SpawnRate=60, Use Spawn Probability=false) {module=SpawnRate001}
  ]
  particleSpawnScript [
   &InitializeParticle(
    Color=&MultiplyLinearColor(
     Linear Color=$System.Paint Color
     Scale Factor=&MakeFloat4FromVectorAndFloat(
      "Vector (RGB)"=<1 1 1>
      "Float (Alpha)"=&FloatFromCurve(FloatCurve=DataInterface() {propertyValues {Curve {keys [<at 0.9 1>,<at 1 0>]}}}, CurveIndex=$Emitter.NormalizedLoopAge)
     )
    )
    Lifetime=0.15
    Position Mode=::NewEnumerator0
    Ribbon Width=15
    Ribbon Width Mode=::NewEnumerator1
    Write Material Random=true
   )
   &SkeletalMeshLocation(
    Bone Sampling Mode=::NewEnumerator4
    Interpolate Mesh Sampling=false
    MaskSpawnGroup=false
    Orientation Sampling=::NewEnumerator1
    Sampled Position Offset=<0 0 -10>
    Skeletal Mesh=$System.Skeletal Mesh
   ) {module=SkeletalMeshLocation001}
   /SetVariables(Particles.RibbonUVDistance:NiagaraFloat=&Multiply_FloatByInt(Float=1, Integer=&ReturnParticleIDAttributes())) {isSetParameters=true}
  ]
  particleUpdateScript [
   &ParticleState()
   &ScaleColor(
    Scale Alpha=&FloatFromCurve(
     FloatCurve=DataInterface() {propertyValues {Curve {keys [
         {interpMode=RCIM_Cubic, tangentMode=RCTM_User, time=0.01, value=0}
         {interpMode=RCIM_Cubic, tangentMode=RCTM_User, time=0.1, value=1}
         {arriveTangent=0.0281, interpMode=RCIM_Cubic, leaveTangent=0.0281, tangentMode=RCTM_User, time=0.5, value=1}
         <at 1 0 RCIM_Cubic>
        ]}}}
     Scale Curve=&FloatFromCurve(
      FloatCurve=DataInterface() {propertyValues {Curve {keys [
          {interpMode=RCIM_Cubic, tangentMode=RCTM_User, time=0.025, value=0}
          <at 0.1 1>
          {interpMode=RCIM_Cubic, tangentMode=RCTM_User, time=0.9, value=1}
          {tangentMode=RCTM_User, time=0.975, value=0}
         ]}}}
      CurveIndex=$Emitter.NormalizedLoopAge
      Scale Curve=&FloatFromCurve(
       FloatCurve=DataInterface() {propertyValues {Curve {keys [{interpMode=RCIM_Cubic, tangentMode=RCTM_User, time=0.5, value=1}, {interpMode=RCIM_Cubic, tangentMode=RCTM_User, time=1, value=0}]}}}
       CurveIndex=/Game/Effects/Niagara/DynamicInputs/Math/FN_NormalizedParticleCount()
      )
     )
    )
    Scale RGB=&Multiply_VectorByFloat(
     Vector=<1 1 1>
     Float=&FloatFromCurve(FloatCurve=DataInterface() {propertyValues {Curve {keys [
          {arriveTangent=-0.307, interpMode=RCIM_Cubic, leaveTangent=-0.307, tangentMode=RCTM_User, time=0, value=1.1}
          {interpMode=RCIM_Cubic, tangentMode=RCTM_User, time=1, value=1}
         ]}}})
    )
    ScaleA=true
    ScaleRGB=true
   )
   &ScaleRibbonWidth(Ribbon Width Scale=&FloatFromCurve(
     FloatCurve=DataInterface() {propertyValues {Curve {keys [
         <at 0 0 RCIM_Cubic>
         {arriveTangent=1.0811, time=0.05, value=1}
         {arriveTangent=-1.0526, interpMode=RCIM_Cubic, leaveTangent=-1.0526, time=0.95, value=1}
         <at 1 0 RCIM_Cubic>
        ]}}}
     CurveIndex=$Emitter.NormalizedLoopAge
     Scale Curve=&FloatFromCurve(FloatCurve=DataInterface() {propertyValues {Curve {keys [
          {arriveTangent=3.0459, interpMode=RCIM_Cubic, leaveTangent=3.0459, tangentMode=RCTM_User, time=0, value=0.5}
          {interpMode=RCIM_Cubic, tangentMode=RCTM_User, time=0.3, value=1}
          {tangentMode=RCTM_User, time=1, value=0}
         ]}}})
    )) {module=ScaleRibbonWidth002}
  ]
  rendererClasses [&NiagaraRibbonRendererProperties]
  renderers [{rendererClass=&NiagaraRibbonRendererProperties}]
 }
 Splash_Foam {
  emitterUpdateScript [
   &EmitterState(
    Inactive Response=::NewEnumerator1
    Life Cycle Mode=::NewEnumerator1
    Loop Behavior=::NewEnumerator1
    Loop Delay=2.25
    Loop Duration=8
    UseLoopDelay=false
   )
   &SpawnBurst_Instantaneous(Spawn Count=10, Spawn Time=3.45, Use Spawn Probability=false, UseLoopCountLimit=false)
  ]
  particleSpawnScript [
   &InitializeParticle(
    Lifetime Max=0.5
    Lifetime Min=0.3
    Lifetime Mode=::NewEnumerator1
    Sprite Rotation Mode=::NewEnumerator2
    Sprite Size Max=<30 70>
    Sprite Size Min=<15 35>
    Sprite Size Mode=::NewEnumerator2
    UsePositionOffset=false
    Write Material Random=true
   ) {module=InitializeParticle001}
   /Niagara/Modules/Update/SubUV/SubUVAnimation(SubUV Animation Mode=::NewEnumerator1)
   &SkeletalMeshLocation(Bone Sampling Mode=::NewEnumerator4, MaskSpawnGroup=false, Sampled Position Offset=<0 0 -5>, Skeletal Mesh=$System.Skeletal Mesh) {module=SkeletalMeshLocation001}
   &ShapeLocation(
    Cylinder Distribution=::NewEnumerator2
    Cylinder Height=1
    Cylinder Radius=8
    MaskSpawnGroup=false
    Offset=<0 0 -20>
    Offset Mode=::NewEnumerator0
    Shape Origin=$Particles.Position
    Shape Primitive=::NewEnumerator1
    U Position=&RandomRangeFloat(Minimum=0, Maximum=1)
    UseOwnerScale=true
    V Position=&RandomRangeFloat(Minimum=0, Maximum=1)
   )
   &AddVelocity(
    Velocity=&Add_Vector(A=$Particles.ShapeLocation.ShapeNormal, B=&RandomRangeVector(Minimum=<0 0 2>, Maximum=<0 0 3>))
    Velocity Speed Scale=&RandomRangeFloat(Minimum=30, Maximum=60)
   ) {module=AddVelocity001}
   &CameraOffset(Camera Offset Amount=5)
   /SetVariables(Particles.VisibilityTag=&SelectIntFromArray(
     Array Sampling Mode=::NewEnumerator1
     Direct Array Index=?
     Int Selection Array=DataInterface() {propertyValues {IntData [0,1,0]}}
     Random Seed=0
    )) {isSetParameters=true}
  ]
  particleUpdateScript [
   &ParticleState()
   &SpriteRotationRate(Rotation Rate=&RandomRangeFloat(Minimum=-90, Maximum=90))
   &Drag(Drag=2.5, Use Linear Drag=true, Use Rotational Drag=false)
   &GravityForce(Gravity=<0 0 -50>)
   &SolveForcesAndVelocity(Clamp Velocity=false, Limit Acceleration=false)
   &ScaleSpriteSize(Scale Factor=&Vector2DFromFloat(Value=&FloatFromCurve(FloatCurve=DataInterface() {propertyValues {Curve {keys [
          {arriveTangent=1.6115, interpMode=RCIM_Cubic, leaveTangent=1.6115, tangentMode=RCTM_User, time=0, value=0.33}
          {interpMode=RCIM_Cubic, tangentMode=RCTM_User, time=1, value=1}
         ]}}})))
   &DynamicMaterialParameters(
    Index 0 Param 1=&RandomRangeFloat(Minimum=0, Maximum=1)
    Index 0 Param 3=&FloatFromCurve(FloatCurve=DataInterface() {propertyValues {Curve {keys [{interpMode=RCIM_Cubic, tangentMode=RCTM_User, time=0, value=0}, {interpMode=RCIM_Cubic, tangentMode=RCTM_User, time=1, value=1}]}}})
    Index 0 Param 4=&FloatFromCurve(FloatCurve=DataInterface() {propertyValues {Curve {keys [{interpMode=RCIM_Cubic, tangentMode=RCTM_User, time=0, value=0}, {interpMode=RCIM_Cubic, tangentMode=RCTM_User, time=1, value=1}]}}})
    Module.Param0WriteEnabled=true
    Module.Param1WriteEnabled=false
    Module.Param2WriteEnabled=true
    Module.Param3WriteEnabled=true
   )
   &ScaleColor(
    Scale RGB=&VectorFromFloat(Value=&FloatFromCurve(
      FloatCurve=DataInterface() {propertyValues {Curve {keys [
          {interpMode=RCIM_Cubic, tangentMode=RCTM_User, time=0, value=2}
          {arriveTangent=-0.9872, interpMode=RCIM_Cubic, leaveTangent=-0.9872, tangentMode=RCTM_User, time=1, value=1.5}
         ]}}}
      Scale Curve=&FN_FrontEndFloatModifier(Gameplay=1.5, Front End=3)
     ))
    ScaleA=true
    ScaleRGB=true
   )
   /SetVariables(Particles.SpriteSize02:Vector2f=&Multiply_Vector2D(A=$Particles.SpriteSize, B=<1 1.5>)) {isSetParameters=true}
  ]
  rendererClasses [&NiagaraSpriteRendererProperties]
  renderers [
   {rendererClass=&NiagaraSpriteRendererProperties}
   {rendererClass=&NiagaraSpriteRendererProperties, rendererIndex=1}
  ]
 }
}
summary {
 systemName=NS_Emote_IronLilac_Gadget
 emitters [
  {emitterName=PaintTrail, rendererClasses [&NiagaraRibbonRendererProperties]}
  {emitterName=Splash_Foam, rendererClasses [&NiagaraSpriteRendererProperties]}
 ]
 userVariables [
  User.bIsFrontend {type=NiagaraBool}
  User.bIsFrontendPreview {type=NiagaraBool}
 ]
}
