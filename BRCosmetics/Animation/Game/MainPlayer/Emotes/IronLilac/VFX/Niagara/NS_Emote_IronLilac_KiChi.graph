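# Niagara VFX for the IronLilac emote's KiChi variant, painting an ink circle and kanji on the ground in a shared system paint color.
# Staggered bursts and system-age material curves reveal the ink, then flare it at about 3.45 seconds with an emissive mesh ring, rising embers and stacked glow planes offset by spawn group, before fading.
# Foam splashes accompany the painting, and skeletal-mesh anim-trail ribbons tinted by time of day follow motion.
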NiagaraSystem /BRCosmetics/Animation/Game/MainPlayer/Emotes/IronLilac/VFX/Niagara/NS_Emote_IronLilac_KiChi
dependencies {
 usedDataInterfaces [
  &NiagaraDataInterfaceCurve
  &NiagaraDataInterfaceColorCurve
  &NiagaraDataInterfaceArrayInt32
  &NiagaraDataInterfaceMeshRendererInfo
  &NiagaraDataInterfaceSkeletalMesh
 ]
 usedDynamicInputs [
  &RemapRange
  &DistanceBetweenPositions
  &Multiply_FloatByInt
  &SpawnGroup
  &FNTimeOfDayColor
 ]
 usedModules [
  &self:SystemSpawnScript.SystemScriptSource.SystemScriptGraph.NiagaraNodeAssignment_0.SetVariables_#1
  &self:SystemSpawnScript.SystemScriptSource.SystemScriptGraph.NiagaraNodeAssignment_1.SetVariables_#2
  &SystemState
  &self:SystemSpawnScript.SystemScriptSource.SystemScriptGraph.NiagaraNodeAssignment_2.SetVariables_#3
 ]
}
emitters {
 Embers {
  emitterUpdateScript [
   &EmitterState(Life Cycle Mode=::NewEnumerator1, Loop Behavior=::NewEnumerator1, Loop Duration Mode=::NewEnumerator1, UseLoopDelay=false)
   &SpawnBurst_Instantaneous(
    Spawn Count=&RandomRangeInt(Minimum=50, Maximum=75)
    Spawn Time=3.45
    Use Spawn Probability=false
    UseLoopCountLimit=false
   )
  ]
  particleSpawnScript [
   /SetVariables(Particles.PerParticleRandom:NiagaraFloat=&RandomRangeFloat(Minimum=0, Maximum=1)) {isSetParameters=true}
   &InitializeParticle(
    Color=<rgba 0 0.3324 0.9 1>
    Lifetime=&Lerp_Float(A=0.3, B=0.5, Alpha=$Particles.PerParticleRandom)
    Sprite Size=&Multiply_Vector2D(
     A=<1.5 1.5>
     B=&MakeVector2D(
      X=&FloatFromCurve(
       FloatCurve=DataInterface() {propertyValues {Curve {keys [<at 0 1>,<at 0.8 2 RCIM_Constant>,<at 0.95 3 RCIM_Constant>]}}}
       CurveIndex=$Particles.PerParticleRandom
      )
      Y=&FloatFromCurve(
       FloatCurve=DataInterface() {propertyValues {Curve {keys [<at 0 1>,<at 0.8 2 RCIM_Constant>,<at 0.95 3>]}}}
       CurveIndex=$Particles.PerParticleRandom
      )
     )
    )
    Sprite Size Mode=::NewEnumerator4
    UsePositionOffset=false
    Write Material Random=true
   )
   &ShapeLocation(
    Cylinder Distribution=::NewEnumerator2
    Cylinder Height=60
    Cylinder Height Midpoint=0
    Cylinder Radius=&RandomRangeFloat(Minimum=30, Maximum=60)
    MaskSpawnGroup=false
    Rotation Axis=<0 0 1>
    Rotation Mode=::NewEnumerator1
    Shape Primitive=::NewEnumerator1
    U Position=&RandomRangeFloat(Minimum=0, Maximum=1)
    UseOwnerScale=true
    V Position=&RandomRangeFloat(Minimum=0, Maximum=1)
   )
   &AddVelocity(Velocity=<0 0 1>, Velocity Speed Scale=&RandomRangeFloat(Minimum=50, Maximum=200))
  ]
  particleUpdateScript [
   &ParticleState()
   &Drag(
    Drag=&FloatFromCurve(
     FloatCurve=DataInterface() {propertyValues {Curve {keys [
         {interpMode=RCIM_Cubic, tangentMode=RCTM_User, time=0, value=0.8}
         {arriveTangent=0.4778, interpMode=RCIM_Cubic, leaveTangent=0.4778, tangentMode=RCTM_User, time=1, value=1.0375}
        ]}}}
     Scale Curve=&RandomRangeFloat(Minimum=2.5, Maximum=5)
    )
    Use Linear Drag=true
    Use Rotational Drag=false
   )
   &SolveForcesAndVelocity(Clamp Velocity=false, Limit Acceleration=false)
   &ScaleColor(Scale RGB=&MakeVectorFromLinearColorRGB(Color=<rgba 1 1 1 1>), ScaleA=true, ScaleRGB=true)
   &ScaleSpriteSize(Uniform Curve Sprite Scale=DataInterface() {propertyValues {Curve {keys [
        {interpMode=RCIM_Cubic, tangentMode=RCTM_User, time=0, value=1}
        {arriveTangent=-2.8658, interpMode=RCIM_Cubic, leaveTangent=-2.8658, tangentMode=RCTM_User, time=1, value=0}
       ]}}})
   &ScaleSpriteSizeBySpeed(
    Max Scale Factor=<0.5 20>
    Sample Scale Factor By Curve=true
    Scale Factor Curve=DataInterface() {propertyValues {Curve {keys [<at 0 0>, {arriveTangent=1.4989, interpMode=RCIM_Cubic, tangentMode=RCTM_Break, time=1, value=1}]}}}
    Velocity Threshold=300
   )
  ]
  rendererClasses [&NiagaraSpriteRendererProperties]
  renderers [{rendererClass=&NiagaraSpriteRendererProperties}]
 }
 Ink_Circle {
  emitterUpdateScript [
   &EmitterState()
   &SpawnBurst_Instantaneous(Spawn Time=1.417, Use Spawn Probability=false, UseLoopCountLimit=false)
  ]
  particleSpawnScript [
   &InitializeParticle(
    Color=$System.Paint Color
    Lifetime=4
    Sprite Size Mode=::NewEnumerator3
    Uniform Sprite Size=125
    UsePositionOffset=false
    Write Material Random=true
   )
   &SpriteFacingAndAlignment(Alignment Coordinate Space=::Local, Sprite Alignment=<0.5 1 0>, Sprite Facing=<0 0 1>, WriteAlignment=true, WriteFacing=true) {module=SpriteFacingAndAlignment001}
   &CameraOffset(Camera Offset Amount=2.5)
  ]
  particleUpdateScript [
   &ParticleState()
   &DynamicMaterialParameters(
    Index 0 Param 1=&FloatFromCurve(
     FloatCurve=DataInterface() {propertyValues {Curve {keys [
         {arriveTangent=-5.0009, interpMode=RCIM_Cubic, leaveTangent=-5.0009, tangentMode=RCTM_User, time=1.3, value=2}
         {arriveTangent=0.0113, interpMode=RCIM_Cubic, leaveTangent=0.0113, tangentMode=RCTM_User, time=2, value=0}
         {arriveTangent=0.0134, interpMode=RCIM_Cubic, leaveTangent=0.0134, tangentMode=RCTM_User, time=4.8, value=0}
         {arriveTangent=3.1747, interpMode=RCIM_Cubic, leaveTangent=3.1747, tangentMode=RCTM_User, time=6, value=2}
        ]}}}
     CurveIndex=$System.Age
    )
    Index 0 Param 2=&FloatFromCurve(
     FloatCurve=DataInterface() {propertyValues {Curve {keys [
         <at 3.3 1 RCIM_Cubic>
         {arriveTangent=-0.0043, interpMode=RCIM_Cubic, leaveTangent=-0.0043, tangentMode=RCTM_User, time=3.5, value=2}
         {tangentMode=RCTM_User, time=4, value=0.8}
        ]}}}
     CurveIndex=$System.Age
    )
    Module.Index0Param0Write=true
    Module.Index0Param1Write=true
    Module.Index0Param2Write=true
    Module.Index0Param3Write=true
   )
   &ScaleColor(
    Scale Alpha=&FloatFromCurve(
     FloatCurve=DataInterface() {propertyValues {Curve {keys [<at 5 1 RCIM_Cubic>, {arriveTangent=-1.811, interpMode=RCIM_Cubic, leaveTangent=-1.811, tangentMode=RCTM_User, time=6, value=0}]}}}
     CurveIndex=$Emitter.Age
     Scale Curve=3
    )
    ScaleA=true
    ScaleRGB=true
   )
  ]
  rendererClasses [&NiagaraSpriteRendererProperties]
  renderers [{rendererClass=&NiagaraSpriteRendererProperties}]
 }
 Kanji_KiChi {
  emitterUpdateScript [
   &EmitterState()
   &SpawnBurst_Instantaneous(Spawn Time=&Add_Float(A=2.2, B=0), Use Spawn Probability=false, UseLoopCountLimit=false)
  ]
  particleSpawnScript [
   &InitializeParticle(
    Color=$System.Paint Color
    Lifetime=3.5
    Position Offset=<0 -2 0>
    Sprite Size Mode=::NewEnumerator3
    Uniform Sprite Size=135
    UsePositionOffset=true
    Write Material Random=true
   )
   &SpriteFacingAndAlignment(Alignment Coordinate Space=::Local, Sprite Facing=<0 0 1>, WriteAlignment=true, WriteFacing=true) {module=SpriteFacingAndAlignment001}
   &CameraOffset(Camera Offset Amount=2.5)
   /Niagara/Modules/Spawn/Decal/InitialDecalOrientation(Orientation Vector:Vector3f=<0 0 1>, Rotation:Vector3f=<0 0 0.25>, Use Rotation Vector=true)
   /SetVariables(Particles.DecalSize:Vector3f=<135 135 135>) {isSetParameters=true}
  ]
  particleUpdateScript [
   &ParticleState()
   &DynamicMaterialParameters(
    Index 0 Param 1=&FloatFromCurve(
     FloatCurve=DataInterface() {propertyValues {Curve {keys [
         {arriveTangent=3.0233, interpMode=RCIM_Cubic, leaveTangent=3.0233, tangentMode=RCTM_User, time=0, value=0}
         {arriveTangent=0.4666, interpMode=RCIM_Cubic, leaveTangent=0.4666, tangentMode=RCTM_User, time=0.5, value=1}
        ]}}}
     Scale Curve=-1
    )
    Index 0 Param 2=&FloatFromCurve(
     FloatCurve=DataInterface() {propertyValues {Curve {keys [<at 3.4 0 RCIM_Cubic>, {arriveTangent=-0.0043, interpMode=RCIM_Cubic, leaveTangent=-0.0043, tangentMode=RCTM_User, time=3.5, value=1}]}}}
     CurveIndex=$System.Age
    )
    Index 0 Param 3=&FloatFromCurve(
     FloatCurve=DataInterface() {propertyValues {Curve {keys [
         {arriveTangent=-0.0921, interpMode=RCIM_Cubic, leaveTangent=-0.0921, tangentMode=RCTM_User, time=4.8, value=0}
         {arriveTangent=0.0876, interpMode=RCIM_Cubic, leaveTangent=0.0876, tangentMode=RCTM_User, time=5.8, value=1}
        ]}}}
     CurveIndex=$System.Age
    )
    Module.Index0Param0Write=true
    Module.Index0Param1Write=true
    Module.Index0Param2Write=true
    Module.Index0Param3Write=true
   )
   &ScaleColor(
    Scale Alpha=&FloatFromCurve(FloatCurve=DataInterface() {propertyValues {Curve {keys [<at 4.7 1>,<at 5 0>]}}}, CurveIndex=$Emitter.Age)
    Scale RGB=&VectorFromFloat(Value=&FloatFromCurve(
      FloatCurve=DataInterface() {propertyValues {Curve {keys [
          <at 3.4 1 RCIM_Cubic>
          {arriveTangent=-0.0043, interpMode=RCIM_Cubic, leaveTangent=-0.0043, tangentMode=RCTM_User, time=3.5, value=5}
          {tangentMode=RCTM_User, time=4, value=3}
         ]}}}
      CurveIndex=$System.Age
     ))
    ScaleA=false
    ScaleRGB=true
   )
  ]
  rendererClasses [&NiagaraSpriteRendererProperties]
  renderers [{rendererClass=&NiagaraSpriteRendererProperties}]
 }
 Motion_Trails {
  emitterSpawnScript [&FN_RibbonIDOffsetUtility()]
  emitterUpdateScript [
   &EmitterState(Inactive Response=::NewEnumerator1, Life Cycle Mode=::NewEnumerator1, Loop Behavior=::NewEnumerator1, UseLoopDelay=true)
   &FN_SpawnAnimTrail(Number of Anim Trails=2)
  ]
  particleSpawnScript [
   &FN_InitializeAnimTrailBoneIndecies(Second Bone or Socket Index=0)
   &FN_InitializeAnimTrailBoneIndecies(AnimTrail Selection=1, First Bone or Socket Index=1) {module=FN_InitializeAnimTrailBoneIndecies001}
   &FN_InitializeAnimTrail(AnimTrailWidth=10, Color=<rgba 0.5 0.45 0.45 0.25>, Lifetime=0.15, SkeletalMesh=$System.Skeletal Mesh, Write Material Random=false)
   &CameraOffset()
  ]
  particleUpdateScript [
   &SolveForcesAndVelocity(Clamp Velocity=false, Limit Acceleration=false)
   &FN_UpdateAnimTrail(Locked Particle RibbonWidth=1, SkeletalMesh=$System.Skeletal Mesh) {module=FN_UpdateAnimTrail001}
   &ScaleColor(
    Scale Alpha=&Multiply_Float(
     A=&FloatFromCurve(FloatCurve=DataInterface() {propertyValues {Curve {keys [{interpMode=RCIM_Cubic, tangentMode=RCTM_User, time=0, value=1}, {interpMode=RCIM_Cubic, tangentMode=RCTM_User, time=1, value=0}]}}})
     B=&FNTimeOfDayScalar(Night Scale=0.8, Morning Scale=0.9, Evening Scale=0.8)
    )
    Scale RGB=&MakeVectorFromLinearColorRGB(Color=$System.TOD Color)
    ScaleA=true
    ScaleRGB=true
   ) {module=ScaleColor001}
   &ScaleRibbonWidth(Ribbon Width Scale=&FloatFromCurve(FloatCurve=DataInterface() {propertyValues {Curve {keys [
         {interpMode=RCIM_Cubic, tangentMode=RCTM_User, time=0, value=1}
         {arriveTangent=-1.7952, interpMode=RCIM_Cubic, leaveTangent=-1.7952, tangentMode=RCTM_User, time=1, value=0}
        ]}}}))
  ]
  rendererClasses [&NiagaraRibbonRendererProperties]
  renderers [{rendererClass=&NiagaraRibbonRendererProperties}]
 }
 Ring_Emissive {
  emitterUpdateScript [
   &EmitterState(Life Cycle Mode=::NewEnumerator1, Loop Behavior=::NewEnumerator1, Loop Duration=8, UseLoopDelay=false)
   &SpawnBurst_Instantaneous(Spawn Time=3.45, Use Spawn Probability=false, UseLoopCountLimit=false)
  ]
  particleSpawnScript [
   &InitializeParticle(
    Color=$System.Paint Color
    Lifetime=0.8
    Mesh Scale=&Multiply_VectorByFloat(Vector=<0.5 1.1 1.1>, Float=0.475)
    Mesh Scale Mode=::NewEnumerator4
    Position Offset=<0 0 -5>
    UsePositionOffset=true
    Write Material Random=true
   )
   &InitialMeshOrientation(Mesh Orientation Mode=::NewEnumerator2, Orientation Vector=<0 0 1>, Use Rotation Vector=false)
  ]
  particleUpdateScript [
   &ParticleState()
   &ScaleColor(
    Scale Alpha=&FloatFromCurve(FloatCurve=DataInterface() {propertyValues {Curve {keys [
         {arriveTangent=9.0725, interpMode=RCIM_Cubic, leaveTangent=9.0725, tangentMode=RCTM_User, time=0, value=0}
         {arriveTangent=-2.674, interpMode=RCIM_Cubic, leaveTangent=-2.674, tangentMode=RCTM_User, time=0.15, value=1}
         {arriveTangent=-0.0405, interpMode=RCIM_Cubic, leaveTangent=-0.0405, tangentMode=RCTM_User, time=1, value=0}
        ]}}})
    Scale RGB=&VectorFromFloat(Value=&FloatFromCurve(
      FloatCurve=DataInterface() {propertyValues {Curve {keys [
          {arriveTangent=-0.0157, interpMode=RCIM_Cubic, leaveTangent=-0.0157, tangentMode=RCTM_User, time=0, value=1.5}
          {arriveTangent=-0.0405, interpMode=RCIM_Cubic, leaveTangent=-0.0405, tangentMode=RCTM_User, time=1, value=1}
         ]}}}
      Scale Curve=2
     ))
    ScaleA=true
    ScaleRGB=true
   )
   &DynamicMaterialParameters(
    Index 0 Param 1=-58
    Index 0 Param 2=&FloatFromCurve(
     FloatCurve=DataInterface() {propertyValues {Curve {keys [
         {arriveTangent=2.0238, interpMode=RCIM_Cubic, leaveTangent=2.0238, tangentMode=RCTM_User, time=0, value=0}
         {interpMode=RCIM_Cubic, tangentMode=RCTM_User, time=1, value=1}
        ]}}}
     Scale Curve=-1
    )
    Module.Index0Param0Write=true
    Module.Index0Param1Write=true
    Module.Index0Param2Write=true
    Module.Index0Param3Write=true
   ) {module=DynamicMaterialParameters001}
  ]
  rendererClasses [&NiagaraMeshRendererProperties]
  renderers [{rendererClass=&NiagaraMeshRendererProperties, rendererIndex=0}]
 }
 Splash_Foam {
  emitterUpdateScript [
   &EmitterState(Life Cycle Mode=::NewEnumerator1, Loop Behavior=::NewEnumerator1, Loop Delay=2.25, Loop Duration=0.8, UseLoopDelay=true)
   &SpawnRate(SpawnRate=20, Use Spawn Probability=false)
  ]
  particleSpawnScript [
   &InitializeParticle(
    Lifetime Max=0.5
    Lifetime Min=0.2
    Lifetime Mode=::NewEnumerator1
    Mesh Scale Mode=::NewEnumerator3
    Position Offset=<0 0 2>
    Sprite Rotation Mode=::NewEnumerator2
    Sprite Size Mode=::NewEnumerator1
    Uniform Sprite Size Max=50
    Uniform Sprite Size Min=30
    UsePositionOffset=true
    Write Material Random=true
   ) {module=InitializeParticle001}
   /Niagara/Modules/Update/SubUV/SubUVAnimation(SubUV Animation Mode=::NewEnumerator1)
   &SpriteFacingAndAlignment(Alignment Coordinate Space=::Local, Sprite Facing=<0 0 1>, WriteAlignment=true, WriteFacing=true) {module=SpriteFacingAndAlignment001}
   &ShapeLocation(
    Cylinder Distribution=::NewEnumerator2
    Cylinder Height=1
    Cylinder Height Midpoint=1
    Cylinder Radius=&RandomRangeFloat(Minimum=10, Maximum=30)
    MaskSpawnGroup=false
    Rotation Axis=<0 0 1>
    Rotation Mode=::NewEnumerator1
    Shape Primitive=::NewEnumerator1
    U Position=&RandomRangeFloat(Minimum=0, Maximum=1)
    UseOwnerScale=true
    V Position=&RandomRangeFloat(Minimum=0, Maximum=1)
   )
   /SetVariables(Particles.VisibilityTag=&SelectIntFromArray(
     Array Sampling Mode=::NewEnumerator1
     Direct Array Index=?
     Int Selection Array=DataInterface() {propertyValues {IntData [0,1,0]}}
     Random Seed=0
    )) {isSetParameters=true}
  ]
  particleUpdateScript [
   &ParticleState()
   &ScaleSpriteSize(Scale Factor=&Vector2DFromFloat(Value=&FloatFromCurve(FloatCurve=DataInterface() {propertyValues {Curve {keys [
          {arriveTangent=1.6115, interpMode=RCIM_Cubic, leaveTangent=1.6115, tangentMode=RCTM_User, time=0, value=0.33}
          {interpMode=RCIM_Cubic, tangentMode=RCTM_User, time=1, value=1}
         ]}}})))
   &DynamicMaterialParameters(
    Index 0 Param 1=&RandomRangeFloat(Minimum=0, Maximum=1)
    Index 0 Param 3=&FloatFromCurve(FloatCurve=DataInterface() {propertyValues {Curve {keys [{interpMode=RCIM_Cubic, tangentMode=RCTM_User, time=0, value=0}, {interpMode=RCIM_Cubic, tangentMode=RCTM_User, time=1, value=1}]}}})
    Index 0 Param 4=&FloatFromCurve(FloatCurve=DataInterface() {propertyValues {Curve {keys [{interpMode=RCIM_Cubic, tangentMode=RCTM_User, time=0, value=0}, {interpMode=RCIM_Cubic, tangentMode=RCTM_User, time=1, value=1}]}}})
    Module.Param0WriteEnabled=true
    Module.Param1WriteEnabled=false
    Module.Param2WriteEnabled=true
    Module.Param3WriteEnabled=true
   )
   &ScaleColor(
    Scale RGB=&VectorFromFloat(Value=&FloatFromCurve(
      FloatCurve=DataInterface() {propertyValues {Curve {keys [
          {interpMode=RCIM_Cubic, tangentMode=RCTM_User, time=0, value=2}
          {arriveTangent=-0.9872, interpMode=RCIM_Cubic, leaveTangent=-0.9872, tangentMode=RCTM_User, time=1, value=1.5}
         ]}}}
      Scale Curve=&FN_FrontEndFloatModifier(Gameplay=1, Front End=2)
     ))
    ScaleA=true
    ScaleRGB=true
   )
   /SetVariables(Particles.SpriteSize02:Vector2f=&Multiply_Vector2D_ByFloat(Float=1.5, Vector2D=$Particles.SpriteSize)) {isSetParameters=true}
  ]
  rendererClasses [&NiagaraSpriteRendererProperties]
  renderers [
   {rendererClass=&NiagaraSpriteRendererProperties}
   {rendererClass=&NiagaraSpriteRendererProperties, rendererIndex=1}
  ]
 }
 Stacked_Plane_Glow_KiChi {
  emitterUpdateScript [
   &EmitterState(
    Inactive Response=::NewEnumerator1
    Life Cycle Mode=::NewEnumerator1
    Loop Behavior=::NewEnumerator1
    Loop Duration Mode=::NewEnumerator1
    UseLoopDelay=false
   )
   &SpawnBurst_Instantaneous(Spawn Count=10, Spawn Time=3.45, Use Spawn Probability=false, UseLoopCountLimit=false)
  ]
  particleSpawnScript [
   &FN_SetSpawnGroup(Normalize Spawn Group=true, Particles Per Group=10, Pattern=::NewEnumerator0)
   &InitializeParticle(
    Color=$System.Paint Color
    Lifetime=0.5
    Mesh Scale=&Multiply_VectorByFloat(Vector=<0.8 0.5 1>, Float=5)
    Mesh Scale Mode=::NewEnumerator4
    Position Offset=&MakeVector(
     X=0
     Y=-2
     Z=&Add_Float(
      A=0
      B=&FloatFromCurve(
       FloatCurve=DataInterface() {propertyValues {Curve {keys [{tangentMode=RCTM_User, time=0, value=0}, {tangentMode=RCTM_User, time=1, value=1}]}}}
       CurveIndex=$Particles.Normalized Spawn Group
       Scale Curve=30
      )
     )
    )
    Sprite Size Mode=::NewEnumerator3
    Uniform Sprite Size=&Add_Float(
     A=135
     B=&FloatFromCurve(
      FloatCurve=DataInterface() {propertyValues {Curve {keys [{tangentMode=RCTM_User, time=0, value=0}, {tangentMode=RCTM_User, time=1, value=1}]}}}
      CurveIndex=$Particles.Normalized Spawn Group
      Scale Curve=25
     )
    )
    UsePositionOffset=true
    Write Material Random=true
   )
   &SpriteFacingAndAlignment(Alignment Coordinate Space=::Local, Sprite Facing=<0 0 1>, WriteAlignment=true, WriteFacing=true) {module=SpriteFacingAndAlignment001}
   &CameraOffset()
   &DynamicMaterialParameters(
    Index 0 Param 1=&FloatFromCurve(
     FloatCurve=DataInterface() {propertyValues {Curve {keys [{tangentMode=RCTM_User, time=0, value=0}, {tangentMode=RCTM_User, time=1, value=1}]}}}
     CurveIndex=$Particles.Normalized Spawn Group
     Scale Curve=0.2
    )
    Module.Param0WriteEnabled=true
    Module.Param1WriteEnabled=false
    Module.Param2WriteEnabled=false
    Module.Param3WriteEnabled=false
   )
  ]
  particleUpdateScript [
   &ParticleState(Kill Particles When Lifetime Has Elapsed=false)
   /SetVariables(Particles.Color=&MultiplyLinearColor(
     Linear Color=<rgba 0 0.2029 1 1>
     Scale Factor=&MakeFloat4FromVectorAndFloat(
      "Vector (RGB)"=<1 1 1>
      "Float (Alpha)"=&FloatFromCurve(
       FloatCurve=DataInterface() {propertyValues {Curve {keys [{tangentMode=RCTM_User, time=0, value=1}, {tangentMode=RCTM_User, time=1, value=0}]}}}
       CurveIndex=$Particles.Normalized Spawn Group
      )
     )
    )) {isSetParameters=true}
   &ScaleColor(
    Color Value To Scale=$Particles.Color
    Scale Alpha=&FloatFromCurve(FloatCurve=DataInterface() {propertyValues {Curve {keys [
         {arriveTangent=9.0725, interpMode=RCIM_Cubic, leaveTangent=9.0725, tangentMode=RCTM_User, time=0, value=0}
         {arriveTangent=0.0154, interpMode=RCIM_Cubic, leaveTangent=0.0154, tangentMode=RCTM_User, time=0.1, value=1}
         {arriveTangent=-0.0405, interpMode=RCIM_Cubic, leaveTangent=-0.0405, tangentMode=RCTM_User, time=1, value=0}
        ]}}})
    Scale RGB=&VectorFromFloat(Value=1)
    ScaleA=true
    ScaleRGB=true
   )
  ]
  rendererClasses [&NiagaraSpriteRendererProperties]
  renderers [{rendererClass=&NiagaraSpriteRendererProperties}]
 }
}
summary {
 systemName=NS_Emote_IronLilac_KiChi
 emitters [
  {emitterName=Ink_Circle, rendererClasses [&NiagaraSpriteRendererProperties]}
  {emitterName=Ring_Emissive, rendererClasses [&NiagaraMeshRendererProperties]}
  {emitterName=Embers, rendererClasses [&NiagaraSpriteRendererProperties]}
  {emitterName=Motion_Trails, rendererClasses [&NiagaraRibbonRendererProperties]}
  {emitterName=Splash_Foam, rendererClasses [&NiagaraSpriteRendererProperties]}
  {emitterName=Kanji_KiChi, rendererClasses [&NiagaraSpriteRendererProperties]}
  {emitterName=Stacked_Plane_Glow_KiChi, rendererClasses [&NiagaraSpriteRendererProperties]}
 ]
 userVariables [
  User.bIsFrontend {type=NiagaraBool}
  User.bIsFrontendPreview {type=NiagaraBool}
  User.RibbonIDOffset {type=NiagaraInt32}
 ]
}
